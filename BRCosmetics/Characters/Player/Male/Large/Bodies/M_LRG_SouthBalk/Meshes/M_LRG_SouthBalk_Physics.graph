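# Physics asset for the SouthBalk large male player body cosmetic, set up for secondary motion of its coat.
# Small base capsules anchor left and right dyn_coat chains: three segments at front and back, one at each side, front-side and back-side.
# Constraints lock twist and limit swing, with limits widening down the longer chains.
# Pelvis and thigh capsules plus a huge, lowered floor capsule, approximating a ground plane, are unconstrained bodies alongside the coat.
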PhysicsAsset /BRCosmetics/Characters/Player/Male/Large/Bodies/M_LRG_SouthBalk/Meshes/M_LRG_SouthBalk_Physics
bodies {
 base_coatBack01_lf [{shapeType=Capsule, length=0.5, radius=0.5, shapeName=base_coatBack01_lf_capsule}]
 base_coatBack01_rt [{shapeType=Capsule, length=0.5, radius=0.5, shapeName=base_coatBack01_rt_capsule}]
 base_coatBackSide01_lf [{shapeType=Capsule, length=0.5, radius=0.5, shapeName=base_coatBackSide01_lf_capsule}]
 base_coatBackSide01_rt [{shapeType=Capsule, length=0.5, radius=0.5, shapeName=base_coatBackSide01_rt_capsule}]
 base_coatFront01_lf [{shapeType=Capsule, length=0.5, radius=0.5, shapeName=base_coatFront01_lf_capsule}]
 base_coatFront01_rt [{shapeType=Capsule, length=0.5, radius=0.5, shapeName=base_coatFront01_rt_capsule}]
 base_coatFrontSide01_lf [{shapeType=Capsule, length=0.5, radius=0.5, shapeName=base_coatFrontSide01_lf_capsule}]
 base_coatFrontSide01_rt [{shapeType=Capsule, length=0.5, radius=0.5, shapeName=base_coatFrontSide01_rt_capsule}]
 base_coatSide01_lf [{shapeType=Capsule, length=0.5, radius=0.5, shapeName=base_coatSide01_lf_capsule}]
 base_coatSide01_rt [{shapeType=Capsule, length=0.5, radius=0.5, shapeName=base_coatSide01_rt_capsule}]
 dyn_coatBack01_lf [{shapeType=Capsule, length=8, radius=1.75, shapeName=dyn_coatBack01_lf_capsule, center=<6.388 0 0.9043>, rotation=<rot 90 0 0>}]
 dyn_coatBack01_rt [{shapeType=Capsule, length=8, radius=1.75, shapeName=dyn_coatBack01_lf_capsule, center=<6.388 0 0.9043>, rotation=<rot 90 0 0>}]
 dyn_coatBack02_lf [{shapeType=Capsule, length=6.5, radius=2, shapeName=dyn_coatBack02_rt_capsule, center=<5.4356 0 0.568>, rotation=<rot 90 0 0>}]
 dyn_coatBack02_rt [{shapeType=Capsule, length=6.5, radius=2, shapeName=dyn_coatBack02_rt_capsule, center=<5.4356 0 0.568>, rotation=<rot 90 0 0>}]
 dyn_coatBack03_lf [{shapeType=Capsule, length=8, radius=2.5, shapeName=dyn_coatBack03_rt_capsule, center=<6.5657 0 0>, rotation=<rot 90 0 0>}]
 dyn_coatBack03_rt [{shapeType=Capsule, length=8, radius=2.5, shapeName=dyn_coatBack03_rt_capsule, center=<6.5657 0 0>, rotation=<rot 90 0 0>}]
 dyn_coatBackSide01_lf [{shapeType=Capsule, length=6.5, radius=2.5, shapeName=dyn_coatFrontSide01_lf_capsule, center=<6.3472 0 0>, rotation=<rot 90 0 0>}]
 dyn_coatBackSide01_rt [{shapeType=Capsule, length=6.5, radius=2.5, shapeName=dyn_coatFrontSide01_lf_capsule, center=<6.3472 0 0>, rotation=<rot 90 0 0>}]
 dyn_coatFront01_lf [{shapeType=Capsule, length=10, radius=4, shapeName=dyn_coatFront01_lf_capsule, center=<4 0 0>, rotation=<rot 0 0 90>}]
 dyn_coatFront01_rt [{shapeType=Capsule, length=10, radius=4, shapeName=dyn_coatFront01_lf_capsule, center=<4 0 0>, rotation=<rot 0 0 90>}]
 dyn_coatFront02_lf [{shapeType=Capsule, length=5, radius=3.75, shapeName=dyn_coatFront02_rt_capsule, center=<4.1382 0 1.2593>, rotation=<rot 0 0 90>}]
 dyn_coatFront02_rt [{shapeType=Capsule, length=5, radius=3.75, shapeName=dyn_coatFront02_rt_capsule, center=<4.1382 0 1.2593>, rotation=<rot 0 0 90>}]
 dyn_coatFront03_lf [{shapeType=Capsule, length=0, radius=3.5, shapeName=dyn_coatFront03_lf_capsule, center=<4.2427 0 0>, rotation=<rot 0 0 90>}]
 dyn_coatFront03_rt [{shapeType=Capsule, length=0, radius=3.5, shapeName=dyn_coatFront03_lf_capsule, center=<4.2427 0 0>, rotation=<rot 0 0 90>}]
 dyn_coatFrontSide01_lf [{shapeType=Capsule, length=5, radius=2.5, shapeName=dyn_coatFrontSide01_lf_capsule, center=<5.1573 0 0>, rotation=<rot 90 0 0>}]
 dyn_coatFrontSide01_rt [{shapeType=Capsule, length=5, radius=2.5, shapeName=dyn_coatFrontSide01_lf_capsule, center=<5.1573 0 0>, rotation=<rot 90 0 0>}]
 dyn_coatSide01_lf [{shapeType=Capsule, length=3.5, radius=2.5, shapeName=dyn_coatFrontSide01_lf_capsule, center=<4.4878 0 0>, rotation=<rot 90 0 0>}]
 dyn_coatSide01_rt [{shapeType=Capsule, length=3.5, radius=2.5, shapeName=dyn_coatFrontSide01_lf_capsule, center=<4.4878 0 0>, rotation=<rot 90 0 0>}]
 floor [{shapeType=Capsule, length=5, radius=510, shapeName=floor_capsule, center=<0 0 -505>}]
 l_thigh_ret [{
   shapeType=Capsule
   length=24.4206
   radius=6.5
   shapeName=l_thigh_ret_capsule
   center=<-16.347 -0.1839 0.0528>
   rotation=<rot -0.0001 -88.3232 -90.315>
  }]
 pelvis [{shapeType=Capsule, length=8, radius=21.5, shapeName=pelvis_capsule, center=<0.7243 1.7736 0>}]
 r_thigh_ret [{
   shapeType=Capsule
   length=24.4206
   radius=6.5
   shapeName=l_thigh_ret_capsule
   center=<16.347 -0.1839 0.0528>
   rotation=<rot -0.0001 -88.3232 -90.315>
  }]
}
constraints [
 {
  bone1Name=dyn_coatBack01_lf
  bone2Name=base_coatBack01_lf
  swing1LimitDegrees=4
  swing2LimitDegrees=5
  twistLimitDegrees=45
  twistMotion=Locked
 }
 {
  bone1Name=dyn_coatBack02_lf
  bone2Name=dyn_coatBack01_lf
  swing1LimitDegrees=12
  swing2LimitDegrees=20
  twistLimitDegrees=15
  twistMotion=Locked
 }
 {
  bone1Name=dyn_coatBack03_lf
  bone2Name=dyn_coatBack02_lf
  swing1LimitDegrees=22
  swing2LimitDegrees=40
  twistLimitDegrees=45
  twistMotion=Locked
 }
 {
  bone1Name=dyn_coatBack01_rt
  bone2Name=base_coatBack01_rt
  swing1LimitDegrees=4
  swing2LimitDegrees=5
  twistLimitDegrees=45
  twistMotion=Locked
 }
 {
  bone1Name=dyn_coatBack02_rt
  bone2Name=dyn_coatBack01_rt
  swing1LimitDegrees=12
  swing2LimitDegrees=20
  twistLimitDegrees=45
  twistMotion=Locked
 }
 {
  bone1Name=dyn_coatBack03_rt
  bone2Name=dyn_coatBack02_rt
  swing1LimitDegrees=22
  swing2LimitDegrees=40
  twistLimitDegrees=45
  twistMotion=Locked
 }
 {
  bone1Name=dyn_coatBackSide01_lf
  bone2Name=base_coatBackSide01_lf
  swing1LimitDegrees=8
  swing2LimitDegrees=10
  twistLimitDegrees=45
  twistMotion=Locked
 }
 {
  bone1Name=dyn_coatBackSide01_rt
  bone2Name=base_coatBackSide01_rt
  swing1LimitDegrees=5
  swing2LimitDegrees=10
  twistLimitDegrees=45
  twistMotion=Locked
 }
 {
  bone1Name=dyn_coatFront01_lf
  bone2Name=base_coatFront01_lf
  swing1LimitDegrees=8
  swing2LimitDegrees=8
  twistLimitDegrees=45
  twistMotion=Locked
 }
 {
  bone1Name=dyn_coatFront02_lf
  bone2Name=dyn_coatFront01_lf
  swing1LimitDegrees=15
  swing2LimitDegrees=25
  twistLimitDegrees=45
  twistMotion=Locked
 }
 {
  bone1Name=dyn_coatFront03_lf
  bone2Name=dyn_coatFront02_lf
  swing1LimitDegrees=25
  swing2LimitDegrees=40
  twistLimitDegrees=45
  twistMotion=Locked
 }
 {
  bone1Name=dyn_coatFront01_rt
  bone2Name=base_coatFront01_rt
  swing1LimitDegrees=8
  swing2LimitDegrees=8
  twistLimitDegrees=45
  twistMotion=Locked
 }
 {
  bone1Name=dyn_coatFront02_rt
  bone2Name=dyn_coatFront01_rt
  swing1LimitDegrees=15
  swing2LimitDegrees=25
  twistLimitDegrees=45
  twistMotion=Locked
 }
 {
  bone1Name=dyn_coatFront03_rt
  bone2Name=dyn_coatFront02_rt
  swing1LimitDegrees=25
  swing2LimitDegrees=40
  twistLimitDegrees=45
  twistMotion=Locked
 }
 {
  bone1Name=dyn_coatFrontSide01_lf
  bone2Name=base_coatFrontSide01_lf
  swing1LimitDegrees=5
  swing2LimitDegrees=10
  twistLimitDegrees=45
  twistMotion=Locked
 }
 {
  bone1Name=dyn_coatFrontSide01_rt
  bone2Name=base_coatFrontSide01_rt
  swing1LimitDegrees=5
  swing2LimitDegrees=10
  twistLimitDegrees=45
  twistMotion=Locked
 }
 {
  bone1Name=dyn_coatSide01_lf
  bone2Name=base_coatSide01_lf
  swing1LimitDegrees=5
  swing2LimitDegrees=10
  twistLimitDegrees=0
  twistMotion=Locked
 }
 {
  bone1Name=dyn_coatSide01_rt
  bone2Name=base_coatSide01_rt
  swing1LimitDegrees=5
  swing2LimitDegrees=10
  twistLimitDegrees=0
  twistMotion=Locked
 }
]
