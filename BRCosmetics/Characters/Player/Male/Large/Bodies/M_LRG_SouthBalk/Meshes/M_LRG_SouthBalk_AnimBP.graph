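# Animation Blueprint for the large male SouthBalk player body, layering secondary motion onto the attached parent's copied pose.
# The pose runs through Control Rig, skydiving-weighted coat-to-pelvis constraints, RigidBody physics simulated relative to spine_02, and additive neck rotations for the front end.
# Updates set rig retargeting from the primary skin weight profile, store parameter-asset gravity results, and toggle an alternative secondary profile at LOD 2+.
# Limb visibility events track gauntlets and boots.
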AnimBlueprint /BRCosmetics/Characters/Player/Male/Large/Bodies/M_LRG_SouthBalk/Meshes/M_LRG_SouthBalk_AnimBP
target_skeleton=&/M_LRG_SouthBalk_Skeleton
graphs [
 AnimGraph {nodes [
   n1 = Control Rig!(
    Source=Copy Pose From Mesh() {bUseAttachedParent=true}
    Alpha
    IsCrouching:bool=false
    IsKneeSliding:bool=Get bIsKneeSliding()
    IsBaseballSliding:bool=Get bIsBaseballSliding()
    GauntletEquipped_Left:bool=Get isWearingGauntlets_Left() -> (isWearingGauntlets_Left:bool)
    GauntletEquipped_Right:bool=Get isWearingGauntlets_Right() -> (isWearingGauntlets_Right:bool)
    IsSkydiving:bool=Get bIsSkyDiving()
    RetargettingDisabled:bool=Get Disable Retargeting() -> (Disable Retargeting:bool)
   ) -> (Pose) {controlRigAssetReference {controlRigAsset=&/M_LRG_SouthBalk_CtrlRig}}
   n2 = Get SkydivingAlpha() -> (SkydivingAlpha:double)
   n3 = "Constraint - base_coatBack01_lf (P)"(ComponentPose=Local To Component(LocalPose=@n1), Alpha=@n2, ConstraintWeights_0=1) {boneToModify {boneName=base_coatBack01_lf}, constraintSetup [{targetBone {boneName=pelvis}}], constraintWeights [1]}
   n4 = "Constraint - base_coatBack01_rt (P)"(ComponentPose=@n3, Alpha=@n2, ConstraintWeights_0=1) {boneToModify {boneName=base_coatBack01_rt}, constraintSetup [{targetBone {boneName=pelvis}}], constraintWeights [1]}
   n5 = RigidBody!(
    ComponentPose=@n4
    Alpha
    OverrideWorldGravity:Vector="0,0,800"
    SimSpaceSettings:Sim Space Settings Structure=(WorldAlpha=1,VelocityScaleZ=1,DampingAlpha=1,MaxLinearVelocity=1100,MaxAngularVelocity=800,MaxLinearAcceleration=2000,MaxAngularAcceleration=200,ExternalLinearDragV=(X=3.5,Y=1.1,Z=1.1),ExternalLinearVelocity=(X=0,Y=0,Z=0),ExternalAngularVelocity=(X=0,Y=0,Z=0))
   ) -> (Pose) {
    bOverrideWorldGravity=true
    overridePhysicsAsset=&/M_LRG_SouthBalk_Physics
    simulationSpace=BaseBoneSpace
    baseBoneRef {boneName=spine_02}
    overrideWorldGravity {z=800}
    simSpaceSettings {
     maxAngularAcceleration=200
     maxAngularVelocity=800
     maxLinearAcceleration=2000
     maxLinearVelocity=1100
     worldAlpha=1
     externalLinearDragV=<3.5 1.1 1.1>
    }
   }
   n6 = Get IsInFrontEnd() -> (IsInFrontEnd:double)
   n7 = "Transform (Modify) Bone - Bone: offset_neckBack_lf"(
    ComponentPose="Transform (Modify) Bone - Bone: offset_neck_lf"(ComponentPose=@n5, Alpha=@n6, Translation="0,0,0", Rotation="0,-8,0", Scale="1,1,1") {rotationMode=BMM_Additive, rotationSpace=BCS_BoneSpace, boneToModify {boneName=offset_neck_lf}, rotation {yaw=-8}}
    Alpha=@n6
    Translation="0,0,0"
    Rotation="0,-5,0"
    Scale="1,1,1"
   ) {rotationMode=BMM_Additive, rotationSpace=BCS_BoneSpace, boneToModify {boneName=offset_neckBack_lf}, rotation {yaw=-5}}
   n8 = "Transform (Modify) Bone - Bone: offset_neckBack_rt"(
    ComponentPose="Transform (Modify) Bone - Bone: offset_neck_rt"(ComponentPose=@n7, Alpha=@n6, Translation="0,0,0", Rotation="0,-5,0", Scale="1,1,1") {rotationMode=BMM_Additive, rotationSpace=BCS_BoneSpace, boneToModify {boneName=offset_neck_rt}, rotation {yaw=-5}}
    Alpha=@n6
    Translation="0,0,0"
    Rotation="0,-5,0"
    Scale="1,1,1"
   ) {rotationMode=BMM_Additive, rotationSpace=BCS_BoneSpace, boneToModify {boneName=offset_neckBack_rt}, rotation {yaw=-5}}
   Output Pose(Result=Component To Local(ComponentPose="Transform (Modify) Bone - Bone: offset_neck_cr"(ComponentPose=@n8, Alpha=@n6, Translation="0,0,0", Rotation="0,-3,0", Scale="1,1,1") {rotationMode=BMM_Additive, rotationSpace=BCS_BoneSpace, boneToModify {boneName=offset_neck_cr}, rotation {yaw=-3}})) {layerGroup=None}
  ]}
 EventGraph {nodes [
   Event BlueprintUpdateAnimation()
   > n1 = Sequence!(execute) -> (then_0, then_1, then_2:Exec)
   n2 = GetPartSkeletalMeshComponent()
   Set Disable Retargeting(
    execute=@n1.then_1
    Disable Retargeting:bool="Not Equal (Name)"(A=GetCurrentSkinWeightProfileName!(self:Skinned Mesh Component Object Reference=@n2, InLayer:ESkinWeightProfileLayer Enum=Primary) -> (ReturnValue:Name=None)) -> (ReturnValue=~)
   ) -> (Output_Get:bool)
   n3 = "Gravity Override State Function (Parameter Asset) (Nativized)"(
    execute=@n1.then_2
    CharacterPartAnimInstance=Self-Reference()
    GameTimeInSeconds=GetGameTimeInSeconds()
    Parameters=&/M_LRG_SouthBalk_Parameters
   )
   > Set GravityOverride(GravityOverride=@n3)
   > Set RBAN Settings Out(RBAN Settings Out=@n3.OutRigidBodyAnimNodeParameters)
   > n4 = Branch(Condition=Get bIsSkyDiving())
   > Set SkydivingAlpha(SkydivingAlpha:double=1) -> (Output_Get:double)
   Set SkydivingAlpha(execute=@n4.else, SkydivingAlpha=0)
   n5 = Branch(execute=@n1.then_0, Condition="integer >= integer"(A=Get CurrentLOD(), B=2))
   > n6 = Sequence!(execute) -> (then_0, then_1)
   > Do Once(Reset=@n8.then_0)
   n7 = GetPartSkeletalMeshComponent()
   > SetSkinWeightProfile!(execute, self:Skinned Mesh Component Object Reference=@n7, InProfileName:Name=Alternative, InLayer:ESkinWeightProfileLayer Enum=Secondary) -> (then, ReturnValue:bool)
   Do Once(execute=@n8.then_1, Reset=@n6.then_1)
   > ClearSkinWeightProfile!(execute, self:Skinned Mesh Component Object Reference=@n7, InLayer:ESkinWeightProfileLayer Enum=Secondary) -> (then)
   n8 = Sequence(execute=@n5.else)
   Event OnLeftLegHidden() {is=title K2Node_Event}
   > Set isWearingBoots_Left(isWearingBoots_Left:bool=true) -> (Output_Get:bool)
   Event OnLeftLegVisible() {is=title K2Node_Event}
   > Set isWearingBoots_Left(isWearingBoots_Left=false)
   Event OnRightLegHidden() {is=title K2Node_Event}
   > Set isWearingBoots_Right(isWearingBoots_Right:bool=true) -> (Output_Get:bool)
   Event OnLeftArmHidden() {is=title K2Node_Event}
   > Set isWearingGauntlets_Left(isWearingGauntlets_Left:bool=true) -> (Output_Get:bool)
   Event OnLeftArmVisible() {is=title K2Node_Event}
   > Set isWearingGauntlets_Left(isWearingGauntlets_Left=false)
   Event OnRightArmVisible() {is=title K2Node_Event}
   > Set isWearingGauntlets_Right(isWearingGauntlets_Right:bool=false) -> (Output_Get:bool)
   Event OnRightArmHidden() {is=title K2Node_Event}
   > Set isWearingGauntlets_Right(isWearingGauntlets_Right=true)
   Event OnRightLegVisible() {is=title K2Node_Event}
   > Set isWearingBoots_Right(isWearingBoots_Right=false)
   Event BlueprintInitializeAnimation()
   > Set IsInFrontEnd(IsInFrontEnd:double="To Float (Boolean)"(InBool=Get bIsInFrontEnd())) -> (Output_Get:double)
  ]}
]
variables [
 Disable Retargeting {type=boolean}
 FrontCoatFix {type=boolean}
 GravityOverride {type=Vector}
 IsInFrontEnd {type=number}
 isWearingBoots_Left {type=boolean}
 isWearingBoots_Right {type=boolean}
 isWearingGauntlets_Left {type=boolean}
 isWearingGauntlets_Right {type=boolean}
 RBAN Settings Out {type=FortRigidBodyAnimNodeParameters}
 SkydivingAlpha {type=number}
]
