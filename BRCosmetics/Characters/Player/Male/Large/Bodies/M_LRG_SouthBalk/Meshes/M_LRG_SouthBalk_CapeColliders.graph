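# Physics asset of collision shapes for the large male SouthBalk player body, named as colliders for its cape.
# Capsules cover the pelvis, spine, clavicles and neck, with a sphere for the head and several overlapping capsules on the upper spine bones.
# Dynamic back-coat bones also carry capsules, so the coat itself adds collision volume.
# Arm, leg and foot bodies have no shapes, and no constraints link any bodies.
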PhysicsAsset /BRCosmetics/Characters/Player/Male/Large/Bodies/M_LRG_SouthBalk/Meshes/M_LRG_SouthBalk_CapeColliders
bodies {
 calf_l []
 calf_r []
 clavicle_l [{shapeType=Capsule, length=11.0192, radius=7.8871, center=<10.947 0.299 6.7051>, rotation=<rot 74.3205 0.4614 -178.9918>}]
 clavicle_r [{shapeType=Capsule, length=11.0192, radius=7.8871, center=<-10.947 -0.299 -6.7051>, rotation=<rot 74.3205 0.4614 1.0082>}]
 dyn_coatBack02_lf [{shapeType=Capsule, length=8.5, radius=8.25, center=<6.9977 1.0559 -3.766>, rotation=<rot 4.9994 0.0002 90>}]
 dyn_coatBack02_rt [{shapeType=Capsule, length=8.5, radius=8.25, center=<6.9977 -1.0559 -3.766>, rotation=<rot 4.9994 0.0002 90>}]
 dyn_coatBack03_lf [{
   shapeType=Capsule
   length=0.3724
   radius=10.4827
   shapeName=dyn_coatBack03_lf_capsule
   center=<6.0068 -0.0895 -4.5982>
   rotation=<rot 1.008 174.6051 -90>
  }]
 dyn_coatBack03_rt [{
   shapeType=Capsule
   length=0.3724
   radius=10.4827
   shapeName=dyn_coatBack03_lf_capsule
   center=<6.0068 0.0895 -4.5982>
   rotation=<rot 1.008 174.6051 -90>
  }]
 foot_l []
 foot_r []
 head [{shapeType=Sphere, radius=17.3125, center=<11.028 -0.174 0>}]
 lowerarm_l []
 lowerarm_r []
 neck_02 [{shapeType=Capsule, length=13.4375, radius=7.3175, center=<-0.2642 -0.3408 0>, rotation=<rot 90 -171.1193 -175.3978>}]
 pelvis [{shapeType=Capsule, length=15.6456, radius=24.5388, center=<-0.824 0.4041 0>}]
 spine_01 [{shapeType=Capsule, length=15.1456, radius=25.0388, center=<-1.7965 2.2926 0>}]
 spine_02 [
  {shapeType=Capsule, length=12, radius=25, center=<-1.711 2.9042 0.0001>}
  {shapeType=Capsule, length=29.3956, radius=11.5807, center=<2.5024 0.4992 3.9825>, rotation=<rot 90 0.0049 0.0049>}
  {shapeType=Capsule, length=29.875, radius=11.1013, center=<2.4704 0.4991 -3.9825>, rotation=<rot 90 -0.0396 -0.0396>}
 ]
 spine_03 [
  {shapeType=Capsule, length=5.5, radius=25, center=<0.1445 3.5853 0.0001>}
  {shapeType=Capsule, length=22.5, radius=7.5, shapeName=spine_04_capsule, center=<6.71 -15.2243 0>}
 ]
 spine_04 [
  {shapeType=Capsule, length=6, radius=22.5, center=<2.766 3.6159 0.0002>}
  {shapeType=Capsule, length=22.5, radius=7.5, shapeName=spine_04_capsule, center=<7.4917 -13.2082 0>}
 ]
 spine_05 [
  {shapeType=Capsule, length=20.5226, radius=14.695, center=<2.5963 1.1228 0>}
  {shapeType=Capsule, length=13.9518, radius=13.9425, center=<-0.2916 0.4452 0>}
 ]
 thigh_l []
 thigh_r []
 upperarm_l []
 upperarm_r []
}
constraints []
