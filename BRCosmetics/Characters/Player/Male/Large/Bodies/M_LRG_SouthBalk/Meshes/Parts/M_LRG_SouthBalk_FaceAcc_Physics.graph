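# Physics asset for the South Balk large male character's face accessory, providing simulated physics for a two-sided mustache.
# Each side chains a small base capsule to three dynamic capsules that grow larger toward the tip.
# Constraints lock twist and allow swing that loosens progressively from root to tip.
# Unconstrained head, clavicle and upper-arm capsules are also included, likely as collision volumes for the mustache.
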PhysicsAsset /BRCosmetics/Characters/Player/Male/Large/Bodies/M_LRG_SouthBalk/Meshes/Parts/M_LRG_SouthBalk_FaceAcc_Physics
bodies {
 base_mustache01_lf [{shapeType=Capsule, length=0.5, radius=0.5}]
 base_mustache01_rt [{shapeType=Capsule, length=0.505, radius=0.505, shapeName=base_mustache01_rt_capsule, rotation=<rot 0 -90 0>}]
 clavicle_l [{shapeType=Capsule, length=18, radius=8.005, shapeName=clavicle_l_capsule, center=<9.6856 0.2761 5.715>, rotation=<rot 80.1788 -1.2133 4.833>}]
 clavicle_r [{
   shapeType=Capsule
   length=18
   radius=8.005
   shapeName=clavicle_l_capsule
   center=<-9.6856 0.2761 -5.715>
   rotation=<rot 80.1788 -1.2133 4.833>
  }]
 dyn_mustache01_lf [{shapeType=Capsule, length=1.35, radius=1.3324, center=<2.3222 -0.1938 0.0565>, rotation=<rot -2.1816 -91.0089 83.2013>}]
 dyn_mustache01_rt [{shapeType=Capsule, length=0.85, radius=1.8324, center=<2.2716 -0.2149 -0.4735>, rotation=<rot -2.1817 91.0089 88.5011>}]
 dyn_mustache02_lf [{shapeType=Capsule, length=1, radius=3.5324, center=<3.6784 0.2883 -0.4746>, rotation=<rot -2.1816 -91.0089 -83.2013>}]
 dyn_mustache02_rt [{shapeType=Capsule, length=1, radius=3.5324, center=<3.6784 0.2883 -0.4746>, rotation=<rot -2.1816 -91.0089 -83.2013>}]
 dyn_mustache03_lf [{shapeType=Capsule, length=1.8926, radius=6.5324, center=<7.811 2.0471 -0.8442>, rotation=<rot -3.3097 -81.2161 -83.6716>}]
 dyn_mustache03_rt [{shapeType=Capsule, length=4.8181, radius=6.4801, center=<8.1623 2.2832 -1.2224>, rotation=<rot -3.9129 -75.6123 -100.8999>}]
 head [{shapeType=Capsule, length=0, radius=11.5, shapeName=head_capsule, center=<6.5096 2.1469 0.8019>}]
 upperarm_l [{
   shapeType=Capsule
   length=19
   radius=7.005
   shapeName=clavicle_l_capsule
   center=<10.5834 -1.383 1.1235>
   rotation=<rot 81.4997 -0.0007 -0.0008>
  }]
 upperarm_r [{
   shapeType=Capsule
   length=19
   radius=7.005
   shapeName=clavicle_l_capsule
   center=<-10.5834 1.383 -1.1235>
   rotation=<rot 81.4997 -0.0007 -0.0008>
  }]
}
constraints [
 {
  bone1Name=dyn_mustache03_lf
  bone2Name=dyn_mustache02_lf
  swing1LimitDegrees=10
  swing2LimitDegrees=10
  twistLimitDegrees=0
  twistMotion=Locked
 }
 {
  bone1Name=dyn_mustache02_lf
  bone2Name=dyn_mustache01_lf
  swing1LimitDegrees=8
  swing2LimitDegrees=8
  twistLimitDegrees=0
  twistMotion=Locked
 }
 {
  bone1Name=dyn_mustache01_lf
  bone2Name=base_mustache01_lf
  swing1LimitDegrees=5
  swing2LimitDegrees=5
  twistLimitDegrees=0
  twistMotion=Locked
 }
 {
  bone1Name=dyn_mustache01_rt
  bone2Name=base_mustache01_rt
  swing1LimitDegrees=5
  swing2LimitDegrees=5
  twistLimitDegrees=0
  twistMotion=Locked
 }
 {
  bone1Name=dyn_mustache02_rt
  bone2Name=dyn_mustache01_rt
  swing1LimitDegrees=8
  swing2LimitDegrees=8
  twistLimitDegrees=0
  twistMotion=Locked
 }
 {
  bone1Name=dyn_mustache03_rt
  bone2Name=dyn_mustache02_rt
  swing1LimitDegrees=10
  swing2LimitDegrees=10
  twistLimitDegrees=0
  twistMotion=Locked
 }
]
