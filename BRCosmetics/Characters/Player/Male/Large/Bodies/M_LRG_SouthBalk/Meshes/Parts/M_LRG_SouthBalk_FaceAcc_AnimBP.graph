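# Anim Blueprint for the M_LRG_SouthBalk face accessory, adding secondary motion to the character's mustache.
# It layers body-mesh shoulder, coat and neck bones over the parent pose, then runs a Control Rig, RigidBody and chained spring-based AnimDynamics on both mustache sides.
# DefaultSlot animation overrides the mustache via the front-end Frontend_Idle_Bypass curve.
# Events cache the body mesh and set variables for zipline and skydiving gravity.
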AnimBlueprint /BRCosmetics/Characters/Player/Male/Large/Bodies/M_LRG_SouthBalk/Meshes/Parts/M_LRG_SouthBalk_FaceAcc_AnimBP
target_skeleton=&/M_LRG_SouthBalk_FaceAcc_Skeleton
graphs [
 AnimGraph {nodes [
   n1 = Layered blend per bone(
    BasePose=Use cached pose 'MustahceDyn'()
    BlendPoses_0=Slot 'DefaultSlot'()
    BlendWeights_0=Get CustomIdleIsPlayingFloat() -> (CustomIdleIsPlayingFloat:float)
   ) {layerSetup [{branchFilters [
       {boneName=base_mustacheRoot_cr}
       {boneName=base_mustache01_lf}
       {boneName=dyn_mustache01_lf}
       {boneName=dyn_mustache02_lf}
       {boneName=dyn_mustache03_lf}
       {boneName=dyn_mustacheEndUp_lf}
       {boneName=dyn_mustacheEndLow01_lf}
       {boneName=dyn_mustache04_lf}
       {boneName=dyn_mustacheMidUp_lf}
       {boneName=dyn_mustacheMidLow02_lf}
       {boneName=dyn_mustacheMidLow01_lf}
      ]}]}
   Output Pose(Result=@n1) {layerGroup=None}
   n2 = Layered blend per bone(BasePose=Copy Pose From Mesh() {bUseAttachedParent=true}, BlendPoses_0=Copy Pose From Mesh(SourceMeshComponent=Get Body_Mesh())) {layerSetup [{branchFilters [
       {boneName=clavicle_l}
       {boneName=clavicle_r}
       {boneName=upperarm_r}
       {boneName=upperarm_l}
       {boneName=base_coat_back_rt}
       {boneName=base_coat_back_lf}
       {boneName=base_coat_straps_back_rt_offset}
       {boneName=base_coat_straps_back_lf_offset}
       {boneName=dyn_coat_05_lf_03}
       {boneName=dyn_coat_05_rt_03}
       {boneName=dyn_coat_05_lf_01}
       {boneName=dyn_coat_05_rt_01}
       {boneName=dyn_coat_05_lf_02}
       {boneName=dyn_coat_05_rt_02}
       {boneName=offset_neckFront_lf}
       {boneName=offset_neck_lf}
      ]}]}
   Save cached pose 'StartPose'(Pose=@n2) {cacheName=StartPose}
   n3 = Control Rig(Source=Use cached pose 'StartPose'()) {controlRigAssetReference {controlRigAsset=&/M_LRG_SouthBalk_FaceAcc_CtrlRig}}
   n4 = RigidBody!(ComponentPose=Local To Component(LocalPose=@n3), Alpha, OverrideWorldGravity:Vector="0,0,-250", ExternalForce) -> (Pose) {
    bOverrideWorldGravity=true
    overridePhysicsAsset=&/M_LRG_SouthBalk_FaceAcc_Physics
    simulationSpace=BaseBoneSpace
    baseBoneRef {boneName=base_mustacheRoot_cr}
    overrideWorldGravity {z=-250}
    simSpaceSettings {dampingAlpha=0, maxAngularAcceleration=70, maxAngularVelocity=50, maxLinearAcceleration=700, maxLinearVelocity=500, worldAlpha=0.8}
   }
   n5 = Anim Dynamics - Bone: dyn_mustache04_lf!(ComponentPose=@n4, Alpha=1, ComponentLinearAccScale:Vector="1,1,1", ComponentLinearVelScale:Vector="1,1,1", ExternalForce="0,0,0") -> (Pose) {
    angularDampingOverride=1
    angularSpringConstant=4500
    bAngularSpring=true
    bOverrideAngularDamping=true
    bUseGravityOverride=true
    gravityScale=0.35
    simulationSpace=BoneRelative
    worldSpaceMinimumScale=0
    boundBone {boneName=dyn_mustache04_lf}
    componentAppliedLinearAccClamp=<2000 2000 2000>
    componentLinearAccScale=<1 1 1>
    componentLinearVelScale=<1 1 1>
    gravityOverride {z=-100}
    physicsBodyDefinitions [{
      boundBone {boneName=dyn_mustache04_lf}
      boxExtents=<20 20 20>
      constraintSetup {angularLimitsMax=<0 18 15>, angularLimitsMin=<0 -18 -15>, angularTarget=<1 0 0>}
      localJointOffset=<2 0 0>
     }]
    relativeSpaceBone {boneName=base_mustacheRoot_cr}
    simSpaceSettings {maxAngularAcceleration=18, maxAngularVelocity=8, simSpaceAngularAlpha=0.8}
   }
   n6 = Anim Dynamics - Bone: dyn_mustacheEndUp_lf!(ComponentPose=@n5, Alpha=1, ComponentLinearAccScale:Vector="1,1,1", ComponentLinearVelScale:Vector="1,1,1", ExternalForce="0,0,0") -> (Pose) {
    angularDampingOverride=1
    angularSpringConstant=4500
    bAngularSpring=true
    bOverrideAngularDamping=true
    bUseGravityOverride=true
    gravityScale=0.35
    simulationSpace=BoneRelative
    worldSpaceMinimumScale=0
    boundBone {boneName=dyn_mustacheEndUp_lf}
    componentAppliedLinearAccClamp=<2000 2000 2000>
    componentLinearAccScale=<1 1 1>
    componentLinearVelScale=<1 1 1>
    gravityOverride {z=-100}
    physicsBodyDefinitions [{
      boundBone {boneName=dyn_mustacheEndUp_lf}
      boxExtents=<20 20 20>
      constraintSetup {angularLimitsMax=<0 8 15>, angularLimitsMin=<0 -8 -15>, angularTarget=<1 0 0>}
      localJointOffset=<2 0 0>
     }]
    relativeSpaceBone {boneName=base_mustacheRoot_cr}
    simSpaceSettings {maxAngularAcceleration=20, maxAngularVelocity=10, simSpaceAngularAlpha=0.8}
   }
   n7 = Anim Dynamics - Bone: dyn_mustacheEndLow01_lf!(ComponentPose=@n6, Alpha=1, ComponentLinearAccScale:Vector="1,1,1", ComponentLinearVelScale:Vector="1,1,1", ExternalForce="0,0,0") -> (Pose) {
    angularDampingOverride=1
    angularSpringConstant=4500
    bAngularSpring=true
    bOverrideAngularDamping=true
    bUseGravityOverride=true
    gravityScale=0.35
    simulationSpace=BoneRelative
    worldSpaceMinimumScale=0
    boundBone {boneName=dyn_mustacheEndLow01_lf}
    componentAppliedLinearAccClamp=<2000 2000 2000>
    componentLinearAccScale=<1 1 1>
    componentLinearVelScale=<1 1 1>
    gravityOverride {z=-100}
    physicsBodyDefinitions [{
      boundBone {boneName=dyn_mustacheEndLow01_lf}
      boxExtents=<20 20 20>
      constraintSetup {angularLimitsMax=<0 8 15>, angularLimitsMin=<0 -8 -15>, angularTarget=<1 0 0>}
      localJointOffset=<2 0 0>
     }]
    relativeSpaceBone {boneName=base_mustacheRoot_cr}
    simSpaceSettings {maxAngularAcceleration=20, maxAngularVelocity=10, simSpaceAngularAlpha=0.8}
   }
   n8 = Anim Dynamics - Bone: dyn_mustacheMidUp_lf!(ComponentPose=@n7, Alpha=1, ComponentLinearAccScale:Vector="1,1,1", ComponentLinearVelScale:Vector="1,1,1", ExternalForce="0,0,0") -> (Pose) {
    angularDampingOverride=1
    angularSpringConstant=4500
    bAngularSpring=true
    bOverrideAngularDamping=true
    bUseGravityOverride=true
    gravityScale=0.35
    simulationSpace=BoneRelative
    worldSpaceMinimumScale=0
    boundBone {boneName=dyn_mustacheMidUp_lf}
    componentAppliedLinearAccClamp=<2000 2000 2000>
    componentLinearAccScale=<1 1 1>
    componentLinearVelScale=<1 1 1>
    gravityOverride {z=-100}
    physicsBodyDefinitions [{
      boundBone {boneName=dyn_mustacheMidUp_lf}
      boxExtents=<20 20 20>
      constraintSetup {angularLimitsMax=<0 6 6>, angularLimitsMin=<0 6 -6>, angularTarget=<1 0 0>}
      localJointOffset=<2 0 0>
     }]
    relativeSpaceBone {boneName=base_mustacheRoot_cr}
    simSpaceSettings {maxAngularAcceleration=20, maxAngularVelocity=10, simSpaceAngularAlpha=0.75}
   }
   n9 = Anim Dynamics - Bone: dyn_mustacheMidLow02_lf!(ComponentPose=@n8, Alpha=1, ComponentLinearAccScale:Vector="1,1,1", ComponentLinearVelScale:Vector="1,1,1", ExternalForce="0,0,0") -> (Pose) {
    angularDampingOverride=1
    angularSpringConstant=4500
    bAngularSpring=true
    bOverrideAngularDamping=true
    bUseGravityOverride=true
    gravityScale=0.35
    simulationSpace=BoneRelative
    worldSpaceMinimumScale=0
    boundBone {boneName=dyn_mustacheMidLow02_lf}
    componentAppliedLinearAccClamp=<2000 2000 2000>
    componentLinearAccScale=<1 1 1>
    componentLinearVelScale=<1 1 1>
    gravityOverride {z=-100}
    physicsBodyDefinitions [{
      boundBone {boneName=dyn_mustacheMidLow02_lf}
      boxExtents=<20 20 20>
      constraintSetup {angularLimitsMax=<0 6 6>, angularLimitsMin=<0 6 -6>, angularTarget=<1 0 0>}
      localJointOffset=<2 0 0>
     }]
    relativeSpaceBone {boneName=base_mustacheRoot_cr}
    simSpaceSettings {maxAngularAcceleration=20, maxAngularVelocity=10, simSpaceAngularAlpha=0.8}
   }
   n10 = Anim Dynamics - Bone: dyn_mustacheMidLow01_lf!(ComponentPose=@n9, Alpha=1, ComponentLinearAccScale:Vector="1,1,1", ComponentLinearVelScale:Vector="1,1,1", ExternalForce="0,0,0") -> (Pose) {
    angularDampingOverride=1
    angularSpringConstant=4500
    bAngularSpring=true
    bOverrideAngularDamping=true
    bUseGravityOverride=true
    gravityScale=0.35
    simulationSpace=BoneRelative
    worldSpaceMinimumScale=0
    boundBone {boneName=dyn_mustacheMidLow01_lf}
    componentAppliedLinearAccClamp=<2000 2000 2000>
    componentLinearAccScale=<1 1 1>
    componentLinearVelScale=<1 1 1>
    gravityOverride {z=-100}
    physicsBodyDefinitions [{
      boundBone {boneName=dyn_mustacheMidLow01_lf}
      boxExtents=<20 20 20>
      constraintSetup {angularLimitsMax=<0 6 6>, angularLimitsMin=<0 6 -6>, angularTarget=<1 0 0>}
      localJointOffset=<2 0 0>
     }]
    relativeSpaceBone {boneName=base_mustacheRoot_cr}
    simSpaceSettings {maxAngularAcceleration=20, maxAngularVelocity=10, simSpaceAngularAlpha=0.75}
   }
   n11 = Anim Dynamics - Bone: dyn_mustache04_rt!(ComponentPose=@n10, Alpha=1, ComponentLinearAccScale:Vector="1,1,1", ComponentLinearVelScale:Vector="1,1,1", ExternalForce="0,0,0") -> (Pose) {
    angularDampingOverride=1
    angularSpringConstant=4500
    bAngularSpring=true
    bOverrideAngularDamping=true
    bUseGravityOverride=true
    gravityScale=0.35
    simulationSpace=BoneRelative
    worldSpaceMinimumScale=0
    boundBone {boneName=dyn_mustache04_rt}
    componentAppliedLinearAccClamp=<2000 2000 2000>
    componentLinearAccScale=<1 1 1>
    componentLinearVelScale=<1 1 1>
    gravityOverride {z=-100}
    physicsBodyDefinitions [{
      boundBone {boneName=dyn_mustache04_rt}
      boxExtents=<20 20 20>
      constraintSetup {angularLimitsMax=<0 18 15>, angularLimitsMin=<0 -18 -15>, angularTarget=<1 0 0>}
      localJointOffset=<2 0 0>
     }]
    relativeSpaceBone {boneName=base_mustacheRoot_cr}
    simSpaceSettings {maxAngularAcceleration=18, maxAngularVelocity=8, simSpaceAngularAlpha=0.8}
   }
   n12 = Anim Dynamics - Bone: dyn_mustacheEndUp_rt!(ComponentPose=@n11, Alpha=1, ComponentLinearAccScale:Vector="1,1,1", ComponentLinearVelScale:Vector="1,1,1", ExternalForce="0,0,0") -> (Pose) {
    angularDampingOverride=1
    angularSpringConstant=4500
    bAngularSpring=true
    bOverrideAngularDamping=true
    bUseGravityOverride=true
    gravityScale=0.35
    simulationSpace=BoneRelative
    worldSpaceMinimumScale=0
    boundBone {boneName=dyn_mustacheEndUp_rt}
    componentAppliedLinearAccClamp=<2000 2000 2000>
    componentLinearAccScale=<1 1 1>
    componentLinearVelScale=<1 1 1>
    gravityOverride {z=-100}
    physicsBodyDefinitions [{
      boundBone {boneName=dyn_mustacheEndUp_rt}
      boxExtents=<20 20 20>
      constraintSetup {angularLimitsMax=<0 8 15>, angularLimitsMin=<0 -8 -15>, angularTarget=<1 0 0>}
      localJointOffset=<2 0 0>
     }]
    relativeSpaceBone {boneName=base_mustacheRoot_cr}
    simSpaceSettings {maxAngularAcceleration=20, maxAngularVelocity=10, simSpaceAngularAlpha=0.8}
   }
   n13 = Anim Dynamics - Bone: dyn_mustacheEndLow01_rt!(ComponentPose=@n12, Alpha=1, ComponentLinearAccScale:Vector="1,1,1", ComponentLinearVelScale:Vector="1,1,1", ExternalForce="0,0,0") -> (Pose) {
    angularDampingOverride=1
    angularSpringConstant=4500
    bAngularSpring=true
    bOverrideAngularDamping=true
    bUseGravityOverride=true
    gravityScale=0.35
    simulationSpace=BoneRelative
    worldSpaceMinimumScale=0
    boundBone {boneName=dyn_mustacheEndLow01_rt}
    componentAppliedLinearAccClamp=<2000 2000 2000>
    componentLinearAccScale=<1 1 1>
    componentLinearVelScale=<1 1 1>
    gravityOverride {z=-100}
    physicsBodyDefinitions [{
      boundBone {boneName=dyn_mustacheEndLow01_rt}
      boxExtents=<20 20 20>
      constraintSetup {angularLimitsMax=<0 8 15>, angularLimitsMin=<0 -8 -15>, angularTarget=<1 0 0>}
      localJointOffset=<2 0 0>
     }]
    relativeSpaceBone {boneName=base_mustacheRoot_cr}
    simSpaceSettings {maxAngularAcceleration=20, maxAngularVelocity=10, simSpaceAngularAlpha=0.8}
   }
   n14 = Anim Dynamics - Bone: dyn_mustacheMidUp_rt!(ComponentPose=@n13, Alpha=1, ComponentLinearAccScale:Vector="1,1,1", ComponentLinearVelScale:Vector="1,1,1", ExternalForce="0,0,0") -> (Pose) {
    angularDampingOverride=1
    angularSpringConstant=4500
    bAngularSpring=true
    bOverrideAngularDamping=true
    bUseGravityOverride=true
    gravityScale=0.35
    simulationSpace=BoneRelative
    worldSpaceMinimumScale=0
    boundBone {boneName=dyn_mustacheMidUp_rt}
    componentAppliedLinearAccClamp=<2000 2000 2000>
    componentLinearAccScale=<1 1 1>
    componentLinearVelScale=<1 1 1>
    gravityOverride {z=-100}
    physicsBodyDefinitions [{
      boundBone {boneName=dyn_mustacheMidUp_rt}
      boxExtents=<20 20 20>
      constraintSetup {angularLimitsMax=<0 6 6>, angularLimitsMin=<0 6 -6>, angularTarget=<1 0 0>}
      localJointOffset=<2 0 0>
     }]
    relativeSpaceBone {boneName=base_mustacheRoot_cr}
    simSpaceSettings {maxAngularAcceleration=20, maxAngularVelocity=10, simSpaceAngularAlpha=0.75}
   }
   n15 = Anim Dynamics - Bone: dyn_mustacheMidLow02_rt!(ComponentPose=@n14, Alpha=1, ComponentLinearAccScale:Vector="1,1,1", ComponentLinearVelScale:Vector="1,1,1", ExternalForce="0,0,0") -> (Pose) {
    angularDampingOverride=1
    angularSpringConstant=4500
    bAngularSpring=true
    bOverrideAngularDamping=true
    bUseGravityOverride=true
    gravityScale=0.35
    simulationSpace=BoneRelative
    worldSpaceMinimumScale=0
    boundBone {boneName=dyn_mustacheMidLow02_rt}
    componentAppliedLinearAccClamp=<2000 2000 2000>
    componentLinearAccScale=<1 1 1>
    componentLinearVelScale=<1 1 1>
    gravityOverride {z=-100}
    physicsBodyDefinitions [{
      boundBone {boneName=dyn_mustacheMidLow02_rt}
      boxExtents=<20 20 20>
      constraintSetup {angularLimitsMax=<0 6 6>, angularLimitsMin=<0 6 -6>, angularTarget=<1 0 0>}
      localJointOffset=<2 0 0>
     }]
    relativeSpaceBone {boneName=base_mustacheRoot_cr}
    simSpaceSettings {maxAngularAcceleration=20, maxAngularVelocity=10, simSpaceAngularAlpha=0.8}
   }
   n16 = Anim Dynamics - Bone: dyn_mustacheMidLow01_rt!(ComponentPose=@n15, Alpha=1, ComponentLinearAccScale:Vector="1,1,1", ComponentLinearVelScale:Vector="1,1,1", ExternalForce="0,0,0") -> (Pose) {
    angularDampingOverride=1
    angularSpringConstant=4500
    bAngularSpring=true
    bOverrideAngularDamping=true
    bUseGravityOverride=true
    gravityScale=0.35
    simulationSpace=BoneRelative
    worldSpaceMinimumScale=0
    boundBone {boneName=dyn_mustacheMidLow01_rt}
    componentAppliedLinearAccClamp=<2000 2000 2000>
    componentLinearAccScale=<1 1 1>
    componentLinearVelScale=<1 1 1>
    gravityOverride {z=-100}
    physicsBodyDefinitions [{
      boundBone {boneName=dyn_mustacheMidLow01_rt}
      boxExtents=<20 20 20>
      constraintSetup {angularLimitsMax=<0 6 6>, angularLimitsMin=<0 6 -6>, angularTarget=<1 0 0>}
      localJointOffset=<2 0 0>
     }]
    relativeSpaceBone {boneName=base_mustacheRoot_cr}
    simSpaceSettings {maxAngularAcceleration=20, maxAngularVelocity=10, simSpaceAngularAlpha=0.75}
   }
   Save cached pose 'MustahceDyn'(Pose=Component To Local(ComponentPose=@n16)) {cacheName=MustahceDyn}
  ]}
 EventGraph {nodes [
   Event BlueprintUpdateAnimation()
   > n1 = Sequence()
   n2 = GetOwnerAnimationInterface!(self:Custom Character Part Anim Instance Object Reference=Self-Reference()) -> (ReturnValue:Fort Player Animation Interface Object Reference)
   > Is Valid(InputObject=@n2)
   > n3 = Set isZipline(isZipline:bool=Break Fort Anim Input Zipline!(FortAnimInput_Zipline:"Fort Anim Input Zipline Structure (by ref)"=GetZiplineInput!(self:Fort Player Animation Interface Object Reference=@n2) -> (ReturnValue:"Fort Anim Input Zipline Structure (by ref)")) -> (
     bIsZiplining:bool=~
     bIsOnAscender:bool=~
     bShouldPlayPivotTransition:bool=~
     LeanYaw:float=~
     PivotBlendDelayRemaining:float=~
     PivotBlendDelay:float=~
     PivotCardinalDirection:EFortCardinalDirection Enum
     WorldVelocityLastTick:Vector=~
    ) {is=title K2Node_BreakStruct}) -> (Output_Get:bool)
   > n4 = Branch(Condition=@n3)
   > Set animDyn on zipline(animDyn on zipline:Vector="0, 0, 0") -> (Output_Get:Vector)
   Set animDyn on zipline(execute=@n4.else, animDyn on zipline="0.25,0.25,0.25")
   Branch(execute=@n1.then_1, Condition=Get bIsInFrontEnd())
   > Set CustomIdleIsPlayingFloat(CustomIdleIsPlayingFloat:float=GetCurveValue(self=GetOwnerFrontEndAnimInstance(), CurveName=Frontend_Idle_Bypass)) -> (Output_Get:float)
   Event OnBeginSkydiving()
   > Set GravityOverride(GravityOverride="0,0,250")
   Event OnEndSkydiving()
   > Set GravityOverride(GravityOverride="0,0,-250")
   Event OnCharacterCustomizationLoaded()
   > n5 = Cast To PlayerPawn_Athena(Object=TryGetPawnOwner())
   > Set Body_Mesh(Body_Mesh=GetSkeletalMeshForPartType(self=@n5))
  ]}
]
variables [
 animDyn on zipline {type=Vector}
 Body_Mesh {type=SkeletalMeshComponent}
 CustomIdleIsPlayingFloat {type=number}
 GravityOverride {type=Vector, default=<0 0 -250>}
 isZipline {type=boolean}
]
